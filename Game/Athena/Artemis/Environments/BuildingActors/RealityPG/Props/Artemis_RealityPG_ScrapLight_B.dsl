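; Scrap light building prop with a colored point light and optional smokestack sprites.
; Construction copies the mesh material's Emissive2 color to the light and smoke material, randomizes smoke panning speed from a stream, and toggles smoke visibility.
; A physics wake sets a flag, under a cosmetic-events check.
; BeginPlay and the flag's OnRep use it to hide and destroy the smoke and clear both component references, without destroying the light.
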(event EventBeginPlay
  (|CallParentFunction)
  (if (Variables|Default|GethasSimulatedPhysics)
    (CallFunction|RemoveLightandSmokeonClient)))

(event BuildingActor|PhysicsSimulationSupport|EventHandlePhysicsObjectAwakeChanged
  (params (bAwake :type Boolean))
  (Networking|CanExecuteCosmeticEvents
    (:True
      (if (Variables|Default|GethasSimulatedPhysics)
        _
        (elif bAwake
          (Variables|Default|SethasSimulatedPhysics true))))))

(fn OnRep_hasSimulatedPhysics
  (Networking|CanExecuteCosmeticEvents
    (:True
      (if (Variables|Default|GethasSimulatedPhysics)
        (CallFunction|RemoveLightandSmokeonClient)))))

(fn RemoveLightAndSmokeOnClient
  (bind _sm_smokestack_sprites_01 (Variables|Default|GetSM_Smokestack_Sprites_01))
  (Networking|CanExecuteCosmeticEvents
    (:True
      (Rendering|SetVisibility _sm_smokestack_sprites_01 false false)
      (Components|DestroyComponent _sm_smokestack_sprites_01)
      (Variables|Default|SetPointLight 0)
      (Variables|Default|SetSM_Smokestack_Sprites_01 0))))

(fn ConstructionScript
  (bind _emissivecolor (Variables|Default|GetEmissiveColor))
  (bind _returnvalue (Math|Conversions|ToLinearColor(Color) _emissivecolor))
  (bind _sm_smokestack_sprites_01 (Variables|Default|GetSM_Smokestack_Sprites_01))
  (bind _use_smoke (Variables|Default|GetUse%20smoke?))
  (|CallParentFunction)
  (Utilities|FlowControl|Sequence
    (:then_0
      (bind _asmaterial_instance_constant (Utilities|Casting|CastToMaterialInstanceConstant (Rendering|Material|GetMaterial (Variables|Debug|GetStaticMeshComponent) 0)))
      (bind _returnvalue_2 (Rendering|Material|GetVectorParameterValue _asmaterial_instance_constant "Emissive2" :!class /Script/Engine.MaterialInstanceConstant))
      (Variables|Default|SetEmissiveColor (Math|Conversions|ToColor(LinearColor) _returnvalue_2))
      (bind _returnvalue_3 (Rendering|Material|CreateDynamicMaterialInstance (Variables|Default|GetSM_Smokestack_Sprites_01) 0 (Rendering|Material|GetMaterial (Variables|Default|GetSM_Smokestack_Sprites_01) 0) "None" :!class /Script/Engine.PrimitiveComponent))
      (Rendering|Material|SetVectorParameterValue _returnvalue_3 "EmissiveColor" _returnvalue :!class /Script/Engine.MaterialInstanceDynamic)
      (Rendering|Material|SetVectorParameterValue _returnvalue_3 "PannerSpeed" (Math|Conversions|ToLinearColor(Vector) (Math|Vector|MakeVector 0.0 (Math|Float|Lerp 0.1 0.125 (Math|Random|RandomFloatfromStream (Variables|Default|GetStream))))) :!class /Script/Engine.MaterialInstanceDynamic)
      (Rendering|Components|Light|SetLightColor (Variables|Default|GetPointLight) _returnvalue true))
    (:then_1
      ;[1 1 1] Turn off smoke, smoke is off by default
      (Utilities|IsValid _sm_smokestack_sprites_01
        (:Is%20Valid
          (Development|SetHiddeninGame _sm_smokestack_sprites_01 (not _use_smoke) false)
          (Rendering|SetVisibility _sm_smokestack_sprites_01 _use_smoke false))))))
